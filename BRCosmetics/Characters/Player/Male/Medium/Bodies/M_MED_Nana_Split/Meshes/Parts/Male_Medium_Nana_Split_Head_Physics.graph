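# Physics asset for the head part of the split Nana male medium player cosmetic, defining jiggle-style dynamics for the peel and four skull ribs.
# Each dynamic bone has an offset, rotated box body paired with a small box body on its matching base bone.
# Five constraints tie each bone to its base with limited swing and twist, tight for the ribs and looser for the head peel.
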PhysicsAsset /BRCosmetics/Characters/Player/Male/Medium/Bodies/M_MED_Nana_Split/Meshes/Parts/Male_Medium_Nana_Split_Head_Physics
bodies {
 dyn_headPeel [{shapeType=Box, extentX=2.1649, extentY=15.2367, extentZ=15.3666, center=<5.6999 0.387 -0.0357>, rotation=<rot 5.2366 -89.9999 0>}]
 dyn_headPeel_Base [{shapeType=Box, extentX=1.01, extentY=1.01, extentZ=1.01, rotation=<rot 0 -90 0>}]
 dyn_skullRibA [{shapeType=Box, extentX=2.1801, extentY=11.4035, extentZ=6.8582, center=<1.8 0 0.4009>, rotation=<rot 0 -90 0>}]
 dyn_skullRibA_Base [{shapeType=Box, extentX=1.2, extentY=1.2, extentZ=1.2}]
 dyn_skullRibB [{shapeType=Box, extentX=2.1801, extentY=14.0912, extentZ=7.3103, center=<1.8 0 0.3596>, rotation=<rot 3.3399 -89.9999 0>}]
 dyn_skullRibB_Base [{shapeType=Box, extentX=1.2, extentY=1.2, extentZ=1.2}]
 dyn_skullRibC [{shapeType=Box, extentX=2.1801, extentY=16.3579, extentZ=9.5414, center=<1.8 0.3453 -0.0634>, rotation=<rot 10.3453 -89.9997 0>}]
 dyn_skullRibC_Base [{shapeType=Box, extentX=1.2, extentY=1.2, extentZ=1.2}]
 dyn_skullRibD [{shapeType=Box, extentX=2.4789, extentY=16.5127, extentZ=6.8582, center=<1.8 0.2955 -0.1596>, rotation=<rot 28.3543 -89.9999 0>}]
 dyn_skullRibD_Base [{shapeType=Box, extentX=1.2, extentY=1.2, extentZ=1.2}]
}
constraints [
 {bone1Name=dyn_skullRibD, bone2Name=dyn_skullRibD_Base, swing1LimitDegrees=10, swing2LimitDegrees=4, twistLimitDegrees=8}
 {bone1Name=dyn_skullRibC, bone2Name=dyn_skullRibC_Base, swing1LimitDegrees=10, swing2LimitDegrees=4, twistLimitDegrees=8}
 {bone1Name=dyn_skullRibB, bone2Name=dyn_skullRibB_Base, swing1LimitDegrees=10, swing2LimitDegrees=4, twistLimitDegrees=8}
 {bone1Name=dyn_skullRibA, bone2Name=dyn_skullRibA_Base, swing1LimitDegrees=10, swing2LimitDegrees=4, twistLimitDegrees=8}
 {bone1Name=dyn_headPeel, bone2Name=dyn_headPeel_Base, swing1LimitDegrees=30, swing2LimitDegrees=20, twistLimitDegrees=20}
]
