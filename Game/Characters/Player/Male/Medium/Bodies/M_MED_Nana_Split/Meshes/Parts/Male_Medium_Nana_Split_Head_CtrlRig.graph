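# Control Rig for the head part of the Nana Split male medium player body, driving its extra dyn_jaw bone from facial jaw motion.
# It reads C_jaw's local rotation, scales Euler X by 1.5 and applies it to dyn_jaw through ModifyTransforms.
# A separate chain, unconnected to BeginExecution, constrains skull-rib and head-peel base bones to spine_05 and neck_01 with maintained offsets, then repeats the jaw edit.
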ControlRigBlueprint /Game/Characters/Player/Male/Medium/Bodies/M_MED_Nana_Split/Meshes/Parts/Male_Medium_Nana_Split_Head_CtrlRig
graphs [
 RigVMFunctionLibrary {links [], nodes []}
 RigVMModel {nodes [
   n1 = &RigUnit_BeginExecution()
   n2 = &RigUnit_GetBoneTransform(Bone=C_jaw, Space=LocalSpace)
   n3 = &RigVMFunction_MathQuaternionFromEuler(
    Euler=(X=0,Y=0,Z=0)
    Euler.X=&RigVMFunction_MathFloatMul(B=1.5, A=&RigVMFunction_MathQuaternionToEuler(RotationOrder=ZYX, Value=@n2.Transform.Rotation).Result.X) {template="Multiply::Execute(in A,in B,out Result)"}.Result
   )
   &RigUnit_ModifyTransforms(
    ItemToModify=((Item=(Type=Bone,Name="dyn_jaw"),Transform=(Rotation=(X=0,Y=0,Z=0,W=1),Translation=(X=0,Y=0,Z=0),Scale3D=(X=1,Y=1,Z=1))))
    Weight=1
    WeightMinimum=0
    ExecutePin=@n1.ExecutePin
    ItemToModify.0.Transform.Rotation=@n3.Result
   )
   n4 = &RigUnit_GetBoneTransform(Bone=spine_05)
   n5 = &RigUnit_GetBoneTransform(Bone=neck_01)
   /Script/ControlRig.RigUnit_TransformConstraint(
    BaseBone=None
    BaseTransform=(Rotation=(X=0,Y=0,Z=0,W=1),Translation=(X=0,Y=0,Z=0),Scale3D=(X=1,Y=1,Z=1))
    BaseTransformSpace=GlobalSpace
    Bone=dyn_skullRibA_Base
    ExecutePin=()
    Targets="((Transform=(Rotation=(X=0,Y=0,Z=0,W=1),Translation=(X=0,Y=0,Z=0),Scale3D=(X=1,Y=1,Z=1)),Weight=0.8,bMaintainOffset=True,Filter=(TranslationFilter=(bX=True,bY=True,bZ=True),RotationFilter=(bX=True,bY=True,bZ=True),ScaleFilter=(bX=True,bY=True,bZ=True))),(T ...(truncated)"
    bUseInitialTransforms=True
    Targets.0.Transform=@n4.Transform
    Targets.1.Transform=@n5.Transform
   ) {is=struct RigVMUnitNode}
   > &RigUnit_TransformConstraint(
    BaseBone=None
    BaseTransform=(Rotation=(X=0,Y=0,Z=0,W=1),Translation=(X=0,Y=0,Z=0),Scale3D=(X=1,Y=1,Z=1))
    BaseTransformSpace=GlobalSpace
    Bone=dyn_skullRibB_Base
    Targets="((Transform=(Rotation=(X=0,Y=0,Z=0,W=1),Translation=(X=0,Y=0,Z=0),Scale3D=(X=1,Y=1,Z=1)),Weight=0.8,bMaintainOffset=True,Filter=(TranslationFilter=(bX=True,bY=True,bZ=True),RotationFilter=(bX=True,bY=True,bZ=True),ScaleFilter=(bX=True,bY=True,bZ=True))),(T ...(truncated)"
    bUseInitialTransforms=True
    Targets.0.Transform=@n4.Transform
    Targets.1.Transform=@n5.Transform
   )
   > &RigUnit_TransformConstraint(
    BaseBone=None
    BaseTransform=(Rotation=(X=0,Y=0,Z=0,W=1),Translation=(X=0,Y=0,Z=0),Scale3D=(X=1,Y=1,Z=1))
    BaseTransformSpace=GlobalSpace
    Bone=dyn_skullRibC_Base
    Targets="((Transform=(Rotation=(X=0,Y=0,Z=0,W=1),Translation=(X=0,Y=0,Z=0),Scale3D=(X=1,Y=1,Z=1)),Weight=0.8,bMaintainOffset=True,Filter=(TranslationFilter=(bX=True,bY=True,bZ=True),RotationFilter=(bX=True,bY=True,bZ=True),ScaleFilter=(bX=True,bY=True,bZ=True))),(T ...(truncated)"
    bUseInitialTransforms=True
    Targets.0.Transform=@n4.Transform
    Targets.1.Transform=@n5.Transform
   )
   > &RigUnit_TransformConstraint(
    BaseBone=None
    BaseTransform=(Rotation=(X=0,Y=0,Z=0,W=1),Translation=(X=0,Y=0,Z=0),Scale3D=(X=1,Y=1,Z=1))
    BaseTransformSpace=GlobalSpace
    Bone=dyn_skullRibD_Base
    Targets="((Transform=(Rotation=(X=0,Y=0,Z=0,W=1),Translation=(X=0,Y=0,Z=0),Scale3D=(X=1,Y=1,Z=1)),Weight=0.8,bMaintainOffset=True,Filter=(TranslationFilter=(bX=True,bY=True,bZ=True),RotationFilter=(bX=True,bY=True,bZ=True),ScaleFilter=(bX=True,bY=True,bZ=True))),(T ...(truncated)"
    bUseInitialTransforms=True
    Targets.0.Transform=@n4.Transform
    Targets.1.Transform=@n5.Transform
   )
   > &RigUnit_TransformConstraint(
    BaseBone=None
    BaseTransform=(Rotation=(X=0,Y=0,Z=0,W=1),Translation=(X=0,Y=0,Z=0),Scale3D=(X=1,Y=1,Z=1))
    BaseTransformSpace=GlobalSpace
    Bone=dyn_headPeel_Base
    Targets="((Transform=(Rotation=(X=0,Y=0,Z=0,W=1),Translation=(X=0,Y=0,Z=0),Scale3D=(X=1,Y=1,Z=1)),Weight=0.8,bMaintainOffset=True,Filter=(TranslationFilter=(bX=True,bY=True,bZ=True),RotationFilter=(bX=True,bY=True,bZ=True),ScaleFilter=(bX=True,bY=True,bZ=True))),(T ...(truncated)"
    bUseInitialTransforms=True
    Targets.0.Transform=@n4.Transform
    Targets.1.Transform=@n5.Transform
   )
   > /Script/ControlRig.RigUnit_ModifyBoneTransforms(
    BoneToModify=((Bone="dyn_jaw",Transform=(Rotation=(X=0,Y=0,Z=0,W=1),Translation=(X=0,Y=0,Z=0),Scale3D=(X=1,Y=1,Z=1))))
    Mode=AdditiveLocal
    Weight=1
    WeightMaximum=1
    WeightMinimum=0
    BoneToModify.0.Transform.Rotation=@n3.Result
   ) {is=struct RigVMUnitNode}
  ]}
]
hierarchy [
 root {loc=<0 0 0>, rot=<0 0 0 1>}
 attach {parent=root, loc=<0 0 0>, rot=<0 0 0 1>}
 pelvis {parent=root, loc=<0 -0.7071 93.1191>, rot=<0.0127 0.707 -0.0127 -0.707>}
 spine_01 {parent=pelvis, loc=<6.7494 -0.6388 0>, rot=<0 0 0.034 -0.9994>}
 spine_02 {parent=spine_01, loc=<7.2518 -0.035 0>, rot=<0 0 0.0238 -0.9997>}
 spine_03 {parent=spine_02, loc=<6.0104 0 0>, rot=<0 0 0.0149 -0.9999>}
 spine_04 {parent=spine_03, loc=<8.1024 0 0>, rot=<0 0 -0.0132 -0.9999>}
 spine_05 {parent=spine_04, loc=<10.0146 0 0>, rot=<0 0 -0.0015 -1>}
 clavicle_l {parent=spine_05, loc=<4.3575 -0.1466 -2.3223>, rot=<0.0465 -0.7289 -0.009 -0.6829>}
 upperarm_l {parent=clavicle_l, loc=<15.5111 0 0>, rot=<0.0305 -0.536 -0.0057 -0.8437>}
 upperarm_twist_01_l {parent=upperarm_l, loc=<0.0041 0 0>, rot=<0 0 0 1>}
 upperarm_twist_02_l {parent=upperarm_l, loc=<13.7941 0 0>, rot=<0 0 0 1>}
 clavicle_r {parent=spine_05, loc=<4.3575 -0.1466 2.3223>, rot=<0.7289 0.0465 -0.6829 0.009>}
 upperarm_r {parent=clavicle_r, loc=<-15.5111 0 0>, rot=<0.0305 -0.536 -0.0057 -0.8437>}
 upperarm_twist_01_r {parent=upperarm_r, loc=<-0.0041 0 0>, rot=<0 0 0 1>}
 upperarm_twist_02_r {parent=upperarm_r, loc=<-13.7942 0 0>, rot=<0 0 0 1>}
 dyn_elbowHalf {parent=upperarm_r, loc=<-26.6514 0 0>, rot=<0.0391 -0.0099 -0.2453 -0.9686>}
 dyn_scapEnd_Base {parent=clavicle_r, loc=<-14.316 1.6474 -3.012>, rot=<-0.0423 0.0083 0.9183 -0.3934>}
 dyn_scapEnd {parent=dyn_scapEnd_Base, loc=<0 0 0>, rot=<0 0 0 -1>}
 dyn_clavEnd_Base {parent=clavicle_r, loc=<-15.1833 -1.7381 -3.856>, rot=<-0.1122 0.0167 0.9398 0.3224>}
 dyn_clavEnd {parent=dyn_clavEnd_Base, loc=<0 0 0>, rot=<0 0 0 -1>}
 neck_01 {parent=spine_05, loc=<11.3138 0.1529 0>, rot=<0 0 -0.2003 -0.9797>}
 neck_02 {parent=neck_01, loc=<4.5627 0 0>, rot=<0 0 0.0504 -0.9987>}
 head {parent=neck_02, loc=<5.3914 0 0>, rot=<0 0 0.115 -0.9934>}
 faceAttach {parent=head, loc=<-3.9918 -0.3948 0.0161>, rot=<0.0082 -0.7071 -0.0082 -0.7071>, scale=<0.89 0.89 0.89>}
 C_jaw {parent=faceAttach, loc=<0.0181 2.9179 8.7152>, rot=<0 0 0 1>}
 dyn_skullRibD_Base {parent=neck_01, loc=<31.4569 -12.5512 2.944>, rot=<-0.6418 -0.6846 0.1441 0.314>}
 dyn_skullRibD {parent=dyn_skullRibD_Base, loc=<0 0 0>, rot=<0 0 0 1>}
 dyn_skullRibC_Base {parent=neck_01, loc=<37.2683 -11.5851 2.0052>, rot=<-0.6313 -0.7505 0.0717 0.1815>}
 dyn_skullRibC {parent=dyn_skullRibC_Base, loc=<0 0 0>, rot=<0 0 0 1>}
 dyn_skullRibB_Base {parent=neck_01, loc=<42.5466 -9.8833 1.4968>, rot=<-0.5695 -0.8041 0.0494 0.1635>}
 dyn_skullRibB {parent=dyn_skullRibB_Base, loc=<0 0 0>, rot=<0 0 0 1>}
 dyn_skullRibA_Base {parent=neck_01, loc=<46.6529 -7.5507 1.2335>, rot=<-0.5286 -0.8379 0.0322 0.1319>}
 dyn_skullRibA {parent=dyn_skullRibA_Base, loc=<0 0 0>, rot=<0 0 0 1>}
 dyn_headPeel_Base {parent=neck_01, loc=<34.6746 -5.7505 -8.4185>, rot=<-0.7413 -0.2088 -0.0105 -0.6378>}
 dyn_headPeel {parent=dyn_headPeel_Base, loc=<0 0 0>, rot=<0 0 0 -1>}
 dyn_jaw_Base {parent=neck_01, loc=<14.2023 0.5033 0>, rot=<-0.1086 -0.6987 0.1086 -0.6987>}
 dyn_jaw {parent=dyn_jaw_Base, loc=<0 0 0>, rot=<0 0 0 1>}
 dyn_neckRibA {parent=spine_05, loc=<18.7376 -10.918 2.6196>, rot=<-0.0815 -0.0322 -0.844 -0.5292>}
 dyn_neckRibB {parent=spine_05, loc=<12.8583 -11.8247 2.9145>, rot=<-0.0643 -0.0578 -0.814 -0.5745>}
 dyn_clav_Base {parent=spine_05, loc=<2.3385 14.536 1.6274>, rot=<-0.5933 -0.2891 0.7257 0.1944>}
 dyn_clav {parent=dyn_clav_Base, loc=<0 0 0>, rot=<0 0 0 1>}
 dyn_scap_Base {parent=spine_05, loc=<7.1412 -14.9934 3.36>, rot=<0.6587 -0.2388 -0.6541 0.285>}
 dyn_scap {parent=dyn_scap_Base, loc=<0 0 0>, rot=<0 0 0 -1>}
 dyn_ribA_Base {parent=spine_04, loc=<12.7062 0.364 3.4276>, rot=<0.3743 -0.5289 0.6562 0.3868>}
 dyn_ribA {parent=dyn_ribA_Base, loc=<0 0 0>, rot=<0 0 0 1>}
 dyn_ribB_Base {parent=spine_04, loc=<7.8526 0.1887 2.0666>, rot=<0.3844 -0.5319 0.6462 0.3896>}
 dyn_ribB {parent=dyn_ribB_Base, loc=<0 0 0>, rot=<0 0 0 1>}
 dyn_ribC_Base {parent=spine_04, loc=<1.984 -0.4674 1.7195>, rot=<0.4015 -0.5202 0.6329 0.4094>}
 dyn_ribC {parent=dyn_ribC_Base, loc=<0 0 0>, rot=<0 0 0 1>}
 dyn_ribD_Base {parent=spine_04, loc=<-2.8102 -0.6033 0.4655>, rot=<0.4148 -0.5599 0.6099 0.3776>}
 dyn_ribD {parent=dyn_ribD_Base, loc=<0 0 0>, rot=<0 0 0 1>}
 dyn_hipToKnee_r {parent=pelvis, loc=<-9.692 -0.2474 9.4244>, rot=<0.0458 0.059 0.997 0.0205>}
 dyn_bellyJiggle_Base {parent=pelvis, loc=<-5.4061 5.0519 2.7029>, rot=<-0.1061 0.1099 -0.6862 -0.7112>}
 dyn_bellyJiggle {parent=dyn_bellyJiggle_Base, loc=<0 0 0>, rot=<0 0 0 -1>}
]
