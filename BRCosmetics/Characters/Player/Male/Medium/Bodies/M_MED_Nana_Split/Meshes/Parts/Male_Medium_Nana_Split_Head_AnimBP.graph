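# Post-body AnimBP for the Nana Split head part, copying pose and curves from the attached parent mesh and adding secondary motion.
# Weighted constraints drive the head-peel and four skull-rib base bones from neck_01 and spine_05, feeding a RigidBody simulation in spine_05 space with custom gravity.
# A six-target constraint, dominated by spine_05, then places the jaw base.
# The Control Rig node is unconnected and the EventGraph does nothing.
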AnimBlueprint /BRCosmetics/Characters/Player/Male/Medium/Bodies/M_MED_Nana_Split/Meshes/Parts/Male_Medium_Nana_Split_Head_AnimBP
target_skeleton=/Game/Characters/Player/Male/Medium/Bodies/M_MED_Nana_Split/Meshes/Parts/Male_Medium_Nana_Split_Head_Skeleton
graphs [
 AnimGraph {nodes [
   Control Rig() {
    alphaBoolBlend {blendOption=Linear}
    controlRigAssetReference {blueprintRigClass=/Game/Characters/Player/Male/Medium/Bodies/M_MED_Nana_Split/Meshes/Parts/Male_Medium_Nana_Split_Head_CtrlRig.Male_Medium_Nana_Split_Head_CtrlRig_C}
   }
   n1 = "Constraint - dyn_headPeel_Base (PP)"!(
    ComponentPose=Local To Component(LocalPose=Copy Pose From Mesh() {bCopyCurves=true, bUseAttachedParent=true})
    Alpha=1
    ConstraintWeights_0=0.2
    ConstraintWeights_1:float=0.8
   ) -> (Pose) {
    alphaBoolBlend {blendOption=Linear}
    boneToModify {boneName=dyn_headPeel_Base}
    constraintSetup [{targetBone {boneName=neck_01}}, {targetBone {boneName=spine_05}}]
    constraintWeights [1,1]
   }
   n2 = "Constraint - dyn_skullRibA_Base (PP)"!(ComponentPose=@n1, Alpha=1, ConstraintWeights_0=0.2, ConstraintWeights_1:float=0.8) -> (Pose) {
    alphaBoolBlend {blendOption=Linear}
    boneToModify {boneName=dyn_skullRibA_Base}
    constraintSetup [{targetBone {boneName=neck_01}}, {targetBone {boneName=spine_05}}]
    constraintWeights [1,1]
   }
   n3 = "Constraint - dyn_skullRibB_Base (PP)"!(ComponentPose=@n2, Alpha=1, ConstraintWeights_0=0.2, ConstraintWeights_1:float=0.8) -> (Pose) {
    alphaBoolBlend {blendOption=Linear}
    boneToModify {boneName=dyn_skullRibB_Base}
    constraintSetup [{targetBone {boneName=neck_01}}, {targetBone {boneName=spine_05}}]
    constraintWeights [1,1]
   }
   n4 = "Constraint - dyn_skullRibC_Base (PP)"!(ComponentPose=@n3, Alpha=1, ConstraintWeights_0=0.2, ConstraintWeights_1:float=0.8) -> (Pose) {
    alphaBoolBlend {blendOption=Linear}
    boneToModify {boneName=dyn_skullRibC_Base}
    constraintSetup [{targetBone {boneName=neck_01}}, {targetBone {boneName=spine_05}}]
    constraintWeights [1,1]
   }
   n5 = "Constraint - dyn_skullRibD_Base (PP)"!(ComponentPose=@n4, Alpha=1, ConstraintWeights_0=0.2, ConstraintWeights_1:float=0.8) -> (Pose) {
    alphaBoolBlend {blendOption=Linear}
    boneToModify {boneName=dyn_skullRibD_Base}
    constraintSetup [{targetBone {boneName=neck_01}}, {targetBone {boneName=spine_05}}]
    constraintWeights [1,1]
   }
   n6 = RigidBody(ComponentPose=@n5) {
    bClampLinearTranslationLimitToRefPose=true
    bOverrideWorldGravity=true
    overridePhysicsAsset=&/Male_Medium_Nana_Split_Head_Physics
    simulationSpace=BaseBoneSpace
    alphaBoolBlend {blendOption=Linear}
    baseBoneRef {boneName=spine_05}
    componentAppliedLinearAccClamp=<2000 2000 2000>
    componentLinearAccScale=<0.9 0.9 0.9>
    componentLinearVelScale=<0.5 0.5 0.5>
    overrideWorldGravity {z=-500}
   }
   n7 = "Constraint - dyn_jaw_Base (PPPPPP)"!(
    ComponentPose=@n6
    Alpha=1
    ConstraintWeights_0=0.0128
    ConstraintWeights_1:float=0.0192
    ConstraintWeights_2:float=0.0174
    ConstraintWeights_3:float=0.7418
    ConstraintWeights_4:float=0.0537
    ConstraintWeights_5:float=0.1551
   ) -> (Pose) {
    alphaBoolBlend {blendOption=Linear}
    boneToModify {boneName=dyn_jaw_Base}
    constraintSetup [
     {targetBone {boneName=clavicle_l}}
     {targetBone {boneName=spine_03}}
     {targetBone {boneName=spine_04}}
     {targetBone {boneName=spine_05}}
     {targetBone {boneName=clavicle_r}}
     {targetBone {boneName=neck_01}}
    ]
    constraintWeights [1,1,1,1,1,1]
   }
   Output Pose(Result=Component To Local(ComponentPose=@n7)) {layerGroup=None}
  ]}
 EventGraph {nodes [
   TryGetPawnOwner()
   Event BlueprintUpdateAnimation()
  ]}
]
variables []
